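# Looping sound cue for tactical sprint energy drain, in the gadgets and items sound class.
# A looping stamina wave feeds an Enveloper that fades volume in briefly, then a continuous modulator at the root.
# The modulator reads the EnergyPercentage parameter with inverted mappings, so volume and pitch rise as energy falls and the loop is silent at or above the upper input threshold.
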SoundCue /TacticalSprintGame/Audio/TacSprint_EnergyDrainLoop_Cue
root=ModulatorContinuous_0
nodes [
 {
  id=Enveloper_0
  children [WavePlayer_0]
  properties {pitchCurve {editorCurveData {keys [{time=0}]}}, volumeCurve {editorCurveData {keys [<at 0 0>,<at 0.4 1>]}}}
 }
 {
  id=ModulatorContinuous_0
  children [Enveloper_0]
  properties {
   pitchModulationParams {default=0.6, maxInput=0.6, maxOutput=0.8, minOutput=1.2, parameterName=EnergyPercentage}
   volumeModulationParams {default=0.6, maxInput=0.6, maxOutput=0, minOutput=1, parameterName=EnergyPercentage}
  }
 }
 {id=WavePlayer_0, properties {bLooping=true, soundWaveAssetPtr=&/Tactical_Sprint_Stamina_Loop}}
]
properties {maxDistance=2097152, soundClassObject=&Fort_Gadgets_Items, volumeMultiplier=0.5}
