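# Sound cue for the tactical sprint energy-drained effect, playing one of three equally weighted stamina-empty sounds.
# The random pick passes through a short fixed delay, an enveloper that swells then lowers volume while dropping pitch, and a root modulator adding slight variation.
# Concurrency override caps instances per owner, rejecting new plays and setting a retrigger time to avoid spamming.
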SoundCue /TacticalSprintGame/Audio/TacSprint_EnergyDrained_Cue
root=Modulator_0
nodes [
 {id=Delay_0, children [Random_0], properties {delayMax=0.1, delayMin=0.1}}
 {
  id=Enveloper_0
  children [Delay_0]
  properties {
   pitchCurve {editorCurveData {keys [<at 0 1.15>,<at 0.3083 0.8409>,<at 0.5358 0.7058>,<at 0.6404 0.537>]}}
   volumeCurve {editorCurveData {keys [<at 0 0>,<at 0.3859 0.4124>,<at 0.6036 1.0274>,<at 0.8606 0.6467>,<at 1.0105 0.2953>]}}
  }
 }
 {id=Modulator_0, children [Enveloper_0], properties {pitchMax=1.1}}
 {id=Random_0, children [WavePlayer_0,WavePlayer_1,WavePlayer_2], properties {bShouldExcludeFromBranchCulling=true, weights [1,1,1]}}
 {id=WavePlayer_0, properties {soundWaveAssetPtr=&/Tactical_Sprint_Stamina_Empty_01}}
 {id=WavePlayer_1, properties {soundWaveAssetPtr=&/Tactical_Sprint_Stamina_Empty_02}}
 {id=WavePlayer_2, properties {soundWaveAssetPtr=&/Tactical_Sprint_Stamina_Empty_03}}
]
properties {
 bOverrideConcurrency=true
 maxDistance=2097152
 soundClassObject=&Fort_Gadgets_Items
 volumeMultiplier=0.4
 concurrencyOverrides {bLimitToOwner=true, maxCount=2, resolutionRule=PreventNew, retriggerTime=2}
}
