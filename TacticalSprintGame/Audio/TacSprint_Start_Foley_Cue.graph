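# Sound Cue for the first-person foley burst at the start of a tactical sprint, using first-person sprint footstep attenuation.
# A Mixer layers four random picks: tactical gear rustle and a combat-boot sprint step immediately, a scuff shortly after, then a quieter second step.
# The boot layers are randomly pitched up, with the last step highest, so repeated triggers vary.
# All random nodes except the immediate step's are excluded from branch culling.
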SoundCue /TacticalSprintGame/Audio/TacSprint_Start_Foley_Cue
root=Mixer_0
nodes [
 {id=Delay_0, children [Modulator_0], properties {delayMax=0.1, delayMin=0.1}}
 {id=Delay_2, children [Modulator_1], properties {delayMax=0.25, delayMin=0.25}}
 {id=Mixer_0, children [Random_3,Modulator_2,Delay_2,Delay_0], properties {inputVolume [2,2.5,1.4,2]}}
 {id=Modulator_0, children [Random_1], properties {pitchMax=1.2, pitchMin=1.1}}
 {id=Modulator_1, children [Random_0], properties {pitchMax=1.4, pitchMin=1.3, volumeMax=0.9, volumeMin=0.85}}
 {id=Modulator_2, children [Random_4], properties {pitchMax=1.3, pitchMin=1.2}}
 {
  id=Random_0
  children [WavePlayer_0, WavePlayer_1, WavePlayer_4, WavePlayer_6, WavePlayer_14]
  properties {bShouldExcludeFromBranchCulling=true, weights [1,1,1,1,1]}
 }
 {
  id=Random_1
  children [WavePlayer_7, WavePlayer_8, WavePlayer_9, WavePlayer_10, WavePlayer_11]
  properties {bShouldExcludeFromBranchCulling=true, weights [1,1,1,1,1]}
 }
 {
  id=Random_3
  children [WavePlayer_41, WavePlayer_42, WavePlayer_43, WavePlayer_44, WavePlayer_45]
  properties {bShouldExcludeFromBranchCulling=true, weights [1,1,1,1,1]}
 }
 {id=Random_4, children [WavePlayer_2, WavePlayer_3, WavePlayer_5, WavePlayer_12, WavePlayer_13], properties {weights [1,1,1,1,1]}}
 {
  id=WavePlayer_0
  properties {soundWaveAssetPtr=/Game/Sounds/Fort_Player/Footsteps/CombatBoots/Waves/1P/Asphalt/Foley_FS_1P_BootsCombat_Asphalt_Sprint_01}
 }
 {
  id=WavePlayer_1
  properties {soundWaveAssetPtr=/Game/Sounds/Fort_Player/Footsteps/CombatBoots/Waves/1P/Asphalt/Foley_FS_1P_BootsCombat_Asphalt_Sprint_02}
 }
 {
  id=WavePlayer_10
  properties {soundWaveAssetPtr=/Game/Sounds/Fort_Player/Footsteps/CombatBoots/Waves/1P/Asphalt/Foley_FS_1P_BootsCombat_Asphalt_Scuff_06}
 }
 {
  id=WavePlayer_11
  properties {soundWaveAssetPtr=/Game/Sounds/Fort_Player/Footsteps/CombatBoots/Waves/1P/Asphalt/Foley_FS_1P_BootsCombat_Asphalt_Scuff_07}
 }
 {
  id=WavePlayer_12
  properties {soundWaveAssetPtr=/Game/Sounds/Fort_Player/Footsteps/CombatBoots/Waves/1P/Asphalt/Foley_FS_1P_BootsCombat_Asphalt_Sprint_08}
 }
 {
  id=WavePlayer_13
  properties {soundWaveAssetPtr=/Game/Sounds/Fort_Player/Footsteps/CombatBoots/Waves/1P/Asphalt/Foley_FS_1P_BootsCombat_Asphalt_Sprint_10}
 }
 {
  id=WavePlayer_14
  properties {soundWaveAssetPtr=&Foley_FS_1P_BootsCombat_Asphalt_Sprint_10}
 }
 {
  id=WavePlayer_2
  properties {soundWaveAssetPtr=&Foley_FS_1P_BootsCombat_Asphalt_Sprint_01}
 }
 {
  id=WavePlayer_3
  properties {soundWaveAssetPtr=&Foley_FS_1P_BootsCombat_Asphalt_Sprint_02}
 }
 {
  id=WavePlayer_4
  properties {soundWaveAssetPtr=/Game/Sounds/Fort_Player/Footsteps/CombatBoots/Waves/1P/Asphalt/Foley_FS_1P_BootsCombat_Asphalt_Sprint_05}
 }
 {
  id=WavePlayer_41
  properties {soundWaveAssetPtr=/Game/Sounds/Fort_Player/Clothing/TacticalGear/Waves/Foley_Clothing_TacticalGear_CrouchRun_06}
 }
 {
  id=WavePlayer_42
  properties {soundWaveAssetPtr=/Game/Sounds/Fort_Player/Clothing/TacticalGear/Waves/Foley_Clothing_TacticalGear_CrouchRun_07}
 }
 {
  id=WavePlayer_43
  properties {soundWaveAssetPtr=/Game/Sounds/Fort_Player/Clothing/TacticalGear/Waves/Foley_Clothing_TacticalGear_CrouchRun_08}
 }
 {
  id=WavePlayer_44
  properties {soundWaveAssetPtr=/Game/Sounds/Fort_Player/Clothing/TacticalGear/Waves/Foley_Clothing_TacticalGear_CrouchRun_09}
 }
 {
  id=WavePlayer_45
  properties {soundWaveAssetPtr=/Game/Sounds/Fort_Player/Clothing/TacticalGear/Waves/Foley_Clothing_TacticalGear_CrouchRun_10}
 }
 {
  id=WavePlayer_5
  properties {soundWaveAssetPtr=&Foley_FS_1P_BootsCombat_Asphalt_Sprint_05}
 }
 {
  id=WavePlayer_6
  properties {soundWaveAssetPtr=&Foley_FS_1P_BootsCombat_Asphalt_Sprint_08}
 }
 {
  id=WavePlayer_7
  properties {soundWaveAssetPtr=/Game/Sounds/Fort_Player/Footsteps/CombatBoots/Waves/1P/Asphalt/Foley_FS_1P_BootsCombat_Asphalt_Scuff_03}
 }
 {
  id=WavePlayer_8
  properties {soundWaveAssetPtr=/Game/Sounds/Fort_Player/Footsteps/CombatBoots/Waves/1P/Asphalt/Foley_FS_1P_BootsCombat_Asphalt_Scuff_04}
 }
 {
  id=WavePlayer_9
  properties {soundWaveAssetPtr=/Game/Sounds/Fort_Player/Footsteps/CombatBoots/Waves/1P/Asphalt/Foley_FS_1P_BootsCombat_Asphalt_Scuff_05}
 }
]
properties {
 attenuationSettings=/Game/Sounds/Attenuation/Fort_Player_Footsteps_1P_Sprint_Athena_Att
 maxDistance=2097152
 soundClassObject=&Fort_Player_Foley_1P_Athena
 volumeMultiplier=0.6
}
